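# Niagara glow and light effect for the LucidVibe back bling cosmetic.
# A single burst emitter feeds a sprite renderer with custom facing and camera offset, plus a light renderer.
# Glow and light colors derive from system color and brightness, each modulated by a waveform for a subtle pulse.
# Glow alpha also differs between gameplay and front end.
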NiagaraSystem /BRCosmetics/Accessories/FORT_Backpacks/Backpack_F_MED_LucidVibe/FX/NS_Backpack_LucidVibe
dependencies {
 usedDataInterfaces [&NiagaraDataInterfaceCurve]
 usedDynamicInputs [
  &FNTimeOfDayScalar
  &Lerp_Float
  &FloatFromCurve
 ]
 usedModules [
  &self:SystemSpawnScript.SystemScriptSource.SystemScriptGraph.NiagaraNodeAssignment_1.SetVariables_#1
  &SystemState
  &self:SystemSpawnScript.SystemScriptSource.SystemScriptGraph.NiagaraNodeAssignment_0.SetVariables_#2
  &self:SystemSpawnScript.SystemScriptSource.SystemScriptGraph.NiagaraNodeAssignment_2.SetVariables_#3
 ]
}
emitters {GlowAndLight {
  emitterSpawnScript [/SetVariables(Emitter.Position:NiagaraPosition=<0 -18 0>) {isSetParameters=true}]
  emitterUpdateScript [
   &EmitterState(
    Inactive Response=::NewEnumerator1
    Life Cycle Mode=::NewEnumerator1
    Loop Behavior=::NewEnumerator1
    Loop Duration Mode=::NewEnumerator1
    UseLoopDelay=false
   )
   &SpawnBurst_Instantaneous(Use Spawn Probability=false, UseLoopCountLimit=false)
   /SetVariables(
    Emitter.GlowColor:LinearColor=&MultiplyLinearColor(
     Linear Color=$System.Color
     Scale Factor=&MakeFloat4FromVectorAndFloat(
      "Vector (RGB)"=&VectorFromFloat(Value=$System.BrightnessMultiplier)
      "Float (Alpha)"=&Multiply_Float(
       A=&FN_FrontEndFloatModifier(Gameplay=0.09, Front End=0.07)
       B=&Waveform("Waveform [1]"=::NewEnumerator0, "[1] Frequency"=3, "[1] Amplitude Min/Max"=<0.7 1>)
      )
     )
    )
    Emitter.LightColor:LinearColor=&MultiplyLinearColor(
     Linear Color=$System.Color
     Scale Factor=&MakeVector4_FromFloat(Float=&Multiply_Float(
       A=$System.BrightnessMultiplier
       B=&Multiply_Float(A=4, B=&Waveform("Waveform [1]"=::NewEnumerator0, "[1] Frequency"=3, "[1] Amplitude Min/Max"=<0.7 1>))
      ))
    )
   ) {isSetParameters=true}
  ]
  particleSpawnScript [
   &InitializeParticle(
    Color Mode=::NewEnumerator0
    Lifetime=0
    Position Offset=<2.5 -17 4>
    Sprite Size=<85 75>
    Sprite Size Mode=::NewEnumerator4
    UsePositionOffset=true
    Write Material Random=true
   )
   &SpriteFacingAndAlignment(Sprite Facing=<0 -1 0.1>, WriteAlignment=true, WriteFacing=true)
   &CameraOffset(Camera Offset Amount=9)
   &DynamicMaterialParameters(
    Index 0 Param 1=1
    Module.Index0Float0Write=true
    Module.Index0Float1Write=true
    Module.Index0Float2Write=true
    Module.Index0Float3Write=true
   )
  ]
  rendererClasses [&NiagaraSpriteRendererProperties, &NiagaraLightRendererProperties]
  renderers [
   {rendererClass=&NiagaraSpriteRendererProperties}
   {rendererClass=&NiagaraLightRendererProperties, rendererIndex=1}
  ]
 }}
summary {
 systemName=NS_Backpack_LucidVibe
 emitters [{emitterName=GlowAndLight, rendererClasses [&NiagaraSpriteRendererProperties, &NiagaraLightRendererProperties]}]
 userVariables [
  User.bIsFrontEnd {type=NiagaraBool}
  User.bIsFrontEndPreview {type=NiagaraBool}
 ]
}
