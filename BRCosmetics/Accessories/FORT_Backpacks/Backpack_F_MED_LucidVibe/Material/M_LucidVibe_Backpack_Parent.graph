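# Parent material for the Lucid Vibe backpack cosmetic, layering fake highlights and backlighting onto the shared 2023 character material.
# Between the character PreFX and PostFX functions, the Tech Art mask's red channel blackens base color and confines the highlight.
# The highlight combines a stepped camera-to-normal falloff, green-channel edge cutout and sky-light environment sample along the camera vector.
# That highlight and a tinted backlight from the green channel are added to emissive.
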Material /BRCosmetics/Accessories/FORT_Backpacks/Backpack_F_MED_LucidVibe/Material/M_LucidVibe_Backpack_Parent
graph {
 properties {
  bDisableDepthTest=true
  bHasPixelAnimation=true
  bUseMaterialAttributes=true
  bUsedWithMeshDeformer=true
  bUsedWithSkeletalMesh=true
  pixelDepthOffsetMode=PDOM_Legacy
  translucencyLightingMode=TLM_SurfacePerPixelLighting
  usedShadingModels=Default Lit
 }
 nodes [
  n1 = MaterialFunctionCall() {materialFunction=&MF_CharacterParent_2023_PreFX}
  n2 = GetMaterialAttributes(None=@n1.Result) {attributeGetTypes [#1, #2]}
  n3 = ComponentMask(None=@n1.Tech Art Mask) {r=true}
  n4 = ScalarParameter() {defaultValue=0.0456, group=Lucid Vibe - Backpack, parameterName=Fake Highlight - Fresnel Step}
  n5 = CameraVectorWS()
  n6 = VertexNormalWS()
  n7 = SmoothStep(
   Min=@n4
   Max=Add(A=@n4, B=ScalarParameter() {defaultValue=1, group=Lucid Vibe - Backpack, parameterName=Fake Highlight - Fresnel Smoothness})
   Value=Power(Base=DotProduct(A=@n5, B=@n6), Exp=ScalarParameter() {defaultValue=25, group=Lucid Vibe - Backpack, parameterName=Fake Spec - Fresnel Power}) {constExponent=25.3871}
  )
  n8 = ScalarParameter() {defaultValue=0.549, group=Lucid Vibe - Backpack, parameterName=Fake Highlight - Edge Step}
  n9 = ComponentMask(None=@n1.Tech Art Mask) {g=true}
  n10 = SmoothStep(
   Min=@n8
   Max=Add(A=@n8, B=ScalarParameter() {defaultValue=0.5541, group=Lucid Vibe - Backpack, parameterName=Fake Highlight - Edge Smoothness})
   Value=@n9
  )
  n11 = Saturate(None=Subtract(A=Saturate(None=@n7), B=Multiply(A=@n10) {constB=2}))
  n12 = ScalarParameter() {defaultValue=0.5, group=Lucid Vibe - Backpack, parameterName=Fake Highlight - Mask Intensity}
  n13 = CameraVectorWS()
  n14 = Multiply(A=@n3, B=Multiply(A=Multiply(A=@n11, B=@n12), B=SkyLightEnvMapSample(Direction=@n13)))
  n15 = ScalarParameter() {defaultValue=3, group=Lucid Vibe - Backpack, parameterName=Fake Highlight - Spec Intensity}
  n16 = ScalarParameter() {defaultValue=0.5, parameterName=Exp}
  n17 = Add(A=@n2.Emissive Color, B=Multiply(A=Multiply(A=@n14, B=@n15), B=Power(Base=@n9, Exp=@n16)))
  n18 = ScalarParameter() {defaultValue=2.0969, parameterName=Fake Highlight - Backlight Power}
  n19 = Multiply(
   A=Multiply(
    A=Power(Base=@n9, Exp=@n18)
    B=VectorParameter() {parameterName=Fake Highlight - Backlit Color, defaultValue=<rgba 0.2623 0.2307 0.5906 1>}.RGB
   )
   B=ScalarParameter() {defaultValue=1.5, group=Lucid Vibe - Backpack, parameterName=Fake Highlight - Backlight Intensity}
  )
  MaterialFunctionCall(
   Material=SetMaterialAttributes!(
    MaterialAttributes=@n2.MaterialAttributes
    Emissive Color=Add(A=@n17, B=@n19)
    Base Color=LinearInterpolate(A=@n2.Base Color, Alpha=@n3) {constB=0}
   ) {attributeSetTypes [#1, #2]}
   M=@n1.M
   Pre Skinned Local Position=@n1.M
   Distance Blend=@n1.M
   Fresnel=@n1.M
  ) {materialFunction=&MF_CharacterParent_2023_PostFX}
  Multiply(A=ReflectionVectorWS()) {constB=0.5}
 ]
}
