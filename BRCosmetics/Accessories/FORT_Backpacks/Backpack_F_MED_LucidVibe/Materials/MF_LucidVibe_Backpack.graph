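# Material function for the Lucid Vibe backpack that turns masked areas of the input material into a glitchy emissive screen.
# Panned static, a sweep, scanlines and sine-product flicker animate the emissive texture, with camera depth fade blending close/far strengths.
# Tech Art Mask green drives base color and specular overrides; blue shapes edge darkening and a fake view-dependent backlight.
# A fresnel sky-light fake-specular branch is left unconnected.
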MaterialFunction /BRCosmetics/Accessories/FORT_Backpacks/Backpack_F_MED_LucidVibe/Materials/MF_LucidVibe_Backpack
graph {nodes [
  n1 = GetMaterialAttributes(None=FunctionInput() {inputName=Result, inputType=FunctionInput_MaterialAttributes, sortPriority=10}) {attributeGetTypes [#1, #2]}
  n2 = FunctionInput() {inputName=Tech Art Mask}
  n3 = ComponentMask(None=@n2) {g=true}
  n4 = Multiply(A=Add(A=NamedRerouteUsage(), B=NamedRerouteUsage()), B=@n3)
  n5 = Multiply(A=TextureCoordinate(), B=ScalarParameter() {defaultValue=50, group=Lucid Vibe - Backpack, parameterName=Scanline Tiling})
  n6 = ConstantBiasScale(None=Frac(None=Panner(Coordinate=@n5) {speedY=-1})) {bias=-0.5, scale=2}
  n7 = Multiply(
   A=ComponentMask(None=Multiply(A=@n6, B=@n6)) {g=true}
   B=LinearInterpolate(
    A=ScalarParameter() {defaultValue=0.05, group=Lucid Vibe - Backpack, parameterName=Scanline Strength - Close}
    B=ScalarParameter() {defaultValue=0.1, group=Lucid Vibe - Backpack, parameterName=Scanline Strength - Far}
    Alpha=NamedRerouteUsage()
   )
  ) {constB=0.1}
  n8 = NamedRerouteUsage()
  n9 = NamedRerouteUsage()
  n10 = TextureSample(UVs=Add(A=Add(A=@n8, B=@n9), B=TextureCoordinate())) {texture=/BRCosmetics/Accessories/FORT_Backpacks/Backpack_F_MED_LucidVibe/Textures/T_LucidVibe_Backpack_E}
  n11 = VectorParameter() {group=Lucid Vibe - Backpack, parameterName=Emissive Tint, defaultValue=<rgba 0.72 0.8 1 1>}
  n12 = Multiply(
   A=NamedRerouteDeclaration(None=Add(A=@n7)) {name=Scanlines, nodeColor=<rgba 0 0.1562 1 1>}
   B=Multiply(A=Multiply(A=@n10.RGB, B=@n11.RGB), B=ScalarParameter() {defaultValue=1.1, group=Lucid Vibe - Backpack, parameterName=Emissive Intensity})
  )
  n13 = Constant()
  n14 = ScalarParameter() {defaultValue=0.7938, group=Lucid Vibe - Backpack, parameterName=Edge Darkening - Step}
  n15 = ScalarParameter() {defaultValue=0.15, group=Lucid Vibe - Backpack, parameterName=Edge Darkening - Softness}
  n16 = ComponentMask(None=@n2) {b=true}
  n17 = Multiply(
   A=SmoothStep(Min=@n14, Max=Add(A=@n14, B=@n15), Value=@n16)
   B=ScalarParameter() {defaultValue=0.1, group=Lucid Vibe - Backpack, parameterName=Edge Darkening - Intensity}
  )
  n18 = ScalarParameter() {defaultValue=2, group=Lucid Vibe - Backpack, parameterName=Inner Screen - Mask Power}
  n19 = Add(A=Multiply(A=Add(A=@n4, B=@n12), B=NamedRerouteUsage()), B=Multiply(A=Subtract(A=@n13, B=@n17), B=Power(Base=@n16, Exp=@n18)))
  n20 = Power(Base=@n16, Exp=ScalarParameter() {defaultValue=3.4789, group=Lucid Vibe - Backpack, parameterName=Fake Highlight - Backlight Power})
  n21 = VectorParameter() {group=Lucid Vibe - Backpack, parameterName=Fake Highlight - Backlit Color, defaultValue=<rgba 0.2623 0.2307 0.5906 1>}
  n22 = Multiply(
   A=Multiply(
    A=Multiply(A=@n20, B=@n21.RGB)
    B=ScalarParameter() {defaultValue=1.5, group=Lucid Vibe - Backpack, parameterName=Fake Highlight - Backlight Intensity}
   )
   B=NamedRerouteUsage()
  )
  n23 = ScalarParameter() {defaultValue=1.5, group=Lucid Vibe - Backpack, parameterName=Specular}
  FunctionOutput(None=SetMaterialAttributes!(
    MaterialAttributes=@n1.MaterialAttributes
    EmissiveColor=Add(A=@n19, B=@n22)
    BaseColor=LinearInterpolate(A=@n1.BaseColor, Alpha=@n3) {constB=0}
    Specular=LinearInterpolate(A=@n1.Specular, B=@n23, Alpha=@n3) {constB=2}
   ) {attributeSetTypes [#3, #1, #2]}) {outputName=SetMaterialAttributes}
  Multiply(A=Time(), B=ScalarParameter() {defaultValue=3, group=Lucid Vibe - Backpack, parameterName=Flicker Speed})
  n24 = ScalarParameter() {defaultValue=0.35, group=Lucid Vibe - Backpack, parameterName=Fake Highlight - Fresnel Step}
  n25 = ScalarParameter() {defaultValue=1, group=Lucid Vibe - Backpack, parameterName=Fake Highlight - Fresnel Smoothness}
  n26 = DotProduct(A=CameraVectorWS(), B=VertexNormalWS())
  n27 = ScalarParameter() {defaultValue=15.7405, group=Lucid Vibe - Backpack, parameterName=Fake Spec - Fresnel Power}
  n28 = Saturate(None=SmoothStep(Min=@n24, Max=Add(A=@n24, B=@n25), Value=Power(Base=@n26, Exp=@n27) {constExponent=25.3871}))
  n29 = ScalarParameter() {defaultValue=0.9, group=Lucid Vibe - Backpack, parameterName=Fake Highlight - Edge Step}
  n30 = ScalarParameter() {defaultValue=0.75, group=Lucid Vibe - Backpack, parameterName=Fake Highlight - Edge Smoothness}
  n31 = Multiply(A=SmoothStep(Min=@n29, Max=Add(A=@n29, B=@n30), Value=@n16)) {constB=2}
  n32 = Multiply(
   A=Saturate(None=Subtract(A=@n28, B=@n31))
   B=ScalarParameter() {defaultValue=0.5, group=Lucid Vibe - Backpack, parameterName=Fake Highlight - Mask Intensity}
  )
  n33 = SkyLightEnvMapSample(Direction=CameraVectorWS())
  Multiply(
   A=Multiply(A=@n3, B=Multiply(A=@n32, B=@n33))
   B=ScalarParameter() {defaultValue=2.5, group=Lucid Vibe - Backpack, parameterName=Fake Highlight - Spec Intensity}
  )
  n34 = CameraVectorWS()
  n35 = ReflectionVectorWS()
  NamedRerouteDeclaration(None=OneMinus(None=DotProduct(A=@n34, B=@n35))) {name=Backlight Mask, nodeColor=<rgba 0 0.6016 1 1>}
  NamedRerouteDeclaration(None=ScalarParameter() {defaultValue=1, group=Lucid Vibe - Backpack, parameterName=StaticSpeed}) {name=StaticSpeed, nodeColor=<rgba 0 1 0.6484 1>}
  n36 = NamedRerouteUsage()
  n37 = Constant2Vector() {g=-0.05, r=6}
  n38 = NamedRerouteUsage()
  n39 = TextureSample(UVs=Panner(Coordinate=@n36, Speed=Multiply(A=@n37, B=@n38)) {speedX=6, speedY=-0.05}) {
   samplerSource=SSM_Wrap_WorldGroupSettings
   samplerType=SAMPLERTYPE_Masks
   texture=/Game/Missions/Primary/RepairTheShelter/Textures/Cinematic_Screen_Static
  }
  n40 = Constant2Vector() {g=-0.03, r=-3}
  n41 = NamedRerouteUsage()
  n42 = TextureSample(UVs=Panner(Coordinate=@n36, Speed=Multiply(A=@n40, B=@n41)) {speedX=-3, speedY=-0.03}) {
   samplerSource=SSM_Wrap_WorldGroupSettings
   samplerType=SAMPLERTYPE_Masks
   texture=&Cinematic_Screen_Static
  }
  n43 = Multiply(A=@n39.G, B=@n42.R)
  n44 = NamedRerouteUsage()
  n45 = LinearInterpolate(
   A=ScalarParameter() {defaultValue=0.035, group=Lucid Vibe - Backpack, parameterName=Static UV Distort - Close}
   B=ScalarParameter() {defaultValue=0.05, group=Lucid Vibe - Backpack, parameterName=Static UV Distort - Far}
   Alpha=@n44
  )
  NamedRerouteDeclaration(None=AppendVector(A=Multiply(A=@n43, B=@n45), B=Constant())) {name=StaticUVDistrotion, nodeColor=<rgba 1 0.5625 0 1>}
  n46 = Multiply(A=ComponentMask(None=NamedRerouteUsage()) {g=true}) {constB=0.75}
  n47 = Multiply(A=Time()) {constB=-0.2}
  n48 = Power(Base=Frac(None=Add(A=@n46, B=@n47))) {constExponent=20}
  NamedRerouteDeclaration(None=Multiply(A=@n48, B=ScalarParameter() {defaultValue=0.01, group=Lucid Vibe - Backpack, parameterName=Sweep Add}) {constB=0.2}) {name=Sweep, nodeColor=<rgba 0.2422 0 1 1>}
  n49 = LinearInterpolate(
   A=ScalarParameter() {defaultValue=0.001, group=Lucid Vibe - Backpack, parameterName=Sweep UV Distort - Close}
   B=ScalarParameter() {defaultValue=0.005, group=Lucid Vibe - Backpack, parameterName=Sweep UV Distort - Far}
   Alpha=NamedRerouteUsage()
  )
  NamedRerouteDeclaration(None=AppendVector(A=Multiply(A=@n48, B=@n49) {constB=0.02}, B=Constant())) {name=SweepUVDistortion, nodeColor=<rgba 0.0781 1 0 1>}
  n50 = TextureCoordinate()
  n51 = ComponentMask(None=VectorParameter() {group=Lucid Vibe - Backpack, parameterName=TexCoordScale, defaultValue=<rgba 1 1 0 0>}.RGB) {g=true, r=true}
  n52 = VectorParameter() {group=Lucid Vibe - Backpack, parameterName=TexCoordOffset}
  NamedRerouteDeclaration(None=Add(A=Multiply(A=@n50, B=@n51), B=ComponentMask(None=@n52.RGB) {g=true, r=true})) {name=TexCoord, nodeColor=<rgba 1 0 0.4453 1>}
  n53 = Time() {bOverride_Period=true, period=2.5}
  n54 = ScalarParameter() {defaultValue=2, group=Lucid Vibe - Backpack, parameterName=Flicker Speed}
  n55 = Time() {bOverride_Period=true, period=0.4}
  n56 = Multiply(A=Sine(None=Multiply(A=@n53, B=@n54)), B=Sine(None=Multiply(A=@n55, B=@n54)))
  n57 = Sine(None=Multiply(A=Time() {bOverride_Period=true, period=1.13}, B=@n54))
  n58 = Multiply(A=Time() {bOverride_Period=true, period=1.5126}, B=@n54)
  n59 = LinearInterpolate(
   A=ScalarParameter() {defaultValue=0.79, group=Lucid Vibe - Backpack, parameterName=Flicker Intensity}
   Alpha=Multiply(A=Multiply(A=@n56, B=@n57), B=Sine(None=@n58))
  )
  NamedRerouteDeclaration(None=Saturate(None=Frac(None=@n59))) {name=Flicker, nodeColor=<rgba 0 1 1 1>}
  NamedRerouteDeclaration(None=MaterialFunctionCall(
    Fade Length=ScalarParameter() {defaultValue=40, parameterName=Camera - Fade Length}
    Fade Offset=ScalarParameter() {defaultValue=200, parameterName=Camera - Fade Offset}
   ) {materialFunction=&CameraDepthFade}.Result) {name=Distance Fade, nodeColor=<rgba 1 0 0.375 1>}
  n60 = Constant2Vector() {g=0.1, r=5}
  n61 = NamedRerouteUsage()
  n62 = TextureSample(UVs=Panner(Coordinate=@n36, Speed=Multiply(A=@n60, B=@n61)) {speedX=5, speedY=0.1}) {
   samplerSource=SSM_Wrap_WorldGroupSettings
   samplerType=SAMPLERTYPE_Masks
   texture=&Cinematic_Screen_Static
  }
  n63 = Multiply(A=@n62.B) {constB=0.1}
  n64 = ScalarParameter() {defaultValue=0.6, group=Lucid Vibe - Backpack, parameterName=Static Scale - Close}
  n65 = ScalarParameter() {defaultValue=1.7, group=Lucid Vibe - Backpack, parameterName=Static Scale - Far}
  NamedRerouteDeclaration(None=Multiply(A=Add(A=@n63, B=@n43), B=LinearInterpolate(A=@n64, B=@n65, Alpha=@n44))) {name=Static, nodeColor=<rgba 0 1 0.8828 1>}
 ]}
