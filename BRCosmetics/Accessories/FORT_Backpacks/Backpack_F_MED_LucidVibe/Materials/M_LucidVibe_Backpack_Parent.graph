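# Parent material for the LucidVibe backpack cosmetic, a Default Lit material flagged for skeletal meshes, mesh deformers and pixel animation.
# It uses material attributes and holds no shading logic of its own, only three chained material functions.
# The shared 2023 character PreFX function feeds the backpack-specific MF_LucidVibe_Backpack, whose attributes go into the shared PostFX function.
# PreFX also supplies PostFX's other connected inputs directly.
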Material /BRCosmetics/Accessories/FORT_Backpacks/Backpack_F_MED_LucidVibe/Materials/M_LucidVibe_Backpack_Parent
graph {
 properties {
  bDisableDepthTest=true
  bHasPixelAnimation=true
  bUseMaterialAttributes=true
  bUsedWithMeshDeformer=true
  bUsedWithSkeletalMesh=true
  pixelDepthOffsetMode=PDOM_Legacy
  previewMesh=/BRCosmetics/Accessories/FORT_Backpacks/Backpack_F_MED_LucidVibe/Meshes/F_MED_LucidVibe_Pack
  translucencyLightingMode=TLM_SurfacePerPixelLighting
  usedShadingModels=Default Lit
 }
 nodes [
  n1 = MaterialFunctionCall() {materialFunction=&MF_CharacterParent_2023_PreFX}
  MaterialFunctionCall(
   Material=MaterialFunctionCall!(Tech Art Mask=@n1.Tech Art Mask, Result=@n1.Tech Art Mask) {materialFunction=&/MF_LucidVibe_Backpack}.SetMaterialAttributes
   M=@n1.M
   Pre Skinned Local Position=@n1.M
   Distance Blend=@n1.M
   Fresnel=@n1.M
  ) {materialFunction=&MF_CharacterParent_2023_PostFX}
 ]
}
